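# Masked, unlit, two-sided Niagara sprite and mesh material for ash splinters in the turtle shell quest item VFX.
# Two summed noises distort a sphere mask into a ragged shape, cut by a textured directional dissolve and dithered particle alpha.
# Emissive blends particle-tinted Perlin noise toward a hot Color along a thin edge between two offset dissolve thresholds.
# Dynamic parameters supply random UV offset, dissolve progress and edge glow strength.
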Material /BattlePassS33_Quests/Items/QuestItems/Scenario/Shell/Assets/VFX/M_TurtleShell_Splinter_Ashes
graph {
 properties {
  bAutomaticallySetUsageInEditor=false
  bUsedWithNiagaraMeshParticles=true
  bUsedWithNiagaraSprites=true
  blendMode=BLEND_Masked
  shadingModel=MSM_Unlit
  twoSided=true
  usedShadingModels=Unlit
 }
 nodes [
  n1 = TextureObjectParameter() {
   group=DetailNoise
   parameterName=DetailNoise_Texture
   samplerType=SAMPLERTYPE_Masks
   sortPriority=0
   texture=/Game/Global/Textures/Default/Noise/T_Metal_Noise
  }
  n2 = VectorParameter() {group=DetailNoise, parameterName=DetailNoise_UVControl, channelNames {a=VPan, b=UPan, g=VScale, r=UScale}, defaultValue=<rgba 0.5 0.5 0 0>}
  n3 = VectorParameter() {
   group=DetailNoise
   parameterName=DetailNoise_NoiseControl
   sortPriority=60
   channelNames {a=VStrength, b=UStrength, g=VOffset, r=UOffset}
   defaultValue=<rgba -0.5 -0.5 0.5 0.5>
  }
  n4 = TextureObjectParameter() {group=MainNoise, parameterName=MainNoise_Texture, sortPriority=0, texture=/Game/Items/Traps/Textures/T_PerlinRGB_01_M}
  n5 = DynamicParameter() {paramNames [Random,Param2,Param3,Param4]}
  n6 = VectorParameter() {group=MainNoise, parameterName=MainNoise_UVControl, channelNames {a=VPan, b=UPan, g=VScale, r=UScale}, defaultValue=<rgba 0.3 0.3 0 0>}
  n7 = VectorParameter() {
   group=MainNoise
   parameterName=MainNoise_NoiseControl
   sortPriority=60
   channelNames {a=VStrength, b=UStrength, g=VOffset, r=UOffset}
   defaultValue=<rgba -0.5 -0.5 0.5 0.5>
  }
  n8 = Add(
   A=MaterialFunctionCall!(
    NoiseValue=MaterialFunctionCall(Texture=@n1, UVControl_RGB=@n2, UVControl_A=@n2) {materialFunction=&FXCommon_TextureSample}.RG
    Control_RGB=@n3
    Control_A=@n3
   ) {materialFunction=/Game/Effects/Niagara/MasterMaterials/Functions/Common/FXCommon_NoiseControl}.NoiseValue
   B=MaterialFunctionCall(
    NoiseValue=MaterialFunctionCall(Texture=@n4, AdditionalUVOffset=@n5.Random, UVControl_RGB=@n6, UVControl_A=@n6) {materialFunction=&FXCommon_TextureSample}.RG
    Control_RGB=@n7
    Control_A=@n7
   ) {materialFunction=&FXCommon_NoiseControl}.NoiseValue
  )
  n9 = TextureCoordinate()
  n10 = SphereMask(A=Add(A=@n9, B=@n8), B=Constant2Vector() {g=0.5, r=0.5}) {attenuationRadius=0.25, hardnessPercent=0}
  Add(A=@n8, B=@n10)
  DynamicParameter() {parameterIndex=3, paramNames [Random,Param2,Param3,Param4]}
  n11 = TextureObjectParameter() {
   group=MainT
   parameterName=MainT_Texture
   samplerType=SAMPLERTYPE_LinearGrayscale
   sortPriority=0
   texture=&T_SimplePerlin_Greyscale
  }
  n12 = VectorParameter() {group=MainT, parameterName=MainT_UVControl, channelNames {a=VPan, b=UPan, g=VScale, r=UScale}, defaultValue=<rgba 0.2 0.2 0 0>}
  n13 = ParticleColor()
  n14 = Multiply(
   A=Power(
    Base=MaterialFunctionCall(Texture=@n11, AdditionalUVOffset=@n5.Random, UVControl_RGB=@n12, UVControl_A=@n12) {materialFunction=&FXCommon_TextureSample}.R
    Exp=ScalarParameter() {defaultValue=1, parameterName=sP}
   )
   B=@n13.RGB
  )
  n15 = VectorParameter() {parameterName=Color, defaultValue=<rgba 50 0 0.8333 1>}
  n16 = TextureObjectParameter() {
   group=dissolve texture
   parameterName=dissolve texture_Texture
   samplerType=SAMPLERTYPE_LinearColor
   sortPriority=0
   texture=/Engine/Functions/Engine_MaterialFunctions02/ExampleContent/Textures/OffsetNoiseDistanceFields_ForNormals
  }
  n17 = VectorParameter() {
   group=dissolve texture
   parameterName=dissolve texture_UVControl
   channelNames {a=VPan, b=UPan, g=VScale, r=UScale}
   defaultValue=<rgba 0.5 0.5 0 0>
  }
  n18 = MaterialFunctionCall(Texture=@n16, AdditionalUVOffset=@n5.Random, UVControl_RGB=@n17, UVControl_A=@n17) {materialFunction=&FXCommon_TextureSample}
  n19 = TextureCoordinate()
  n20 = MaterialFunctionCall() {materialFunction=&Pi}
  n21 = Multiply(
   A=MaterialFunctionCall(Float2=Rotator(Coordinate=@n19, Time=@n20.Result)) {materialFunction=&BreakOutFloat2Components}.R
   B=ScalarParameter() {defaultValue=1, parameterName=FEWW}
  )
  n22 = Add(A=@n21, B=@n5.Param2)
  n23 = ScalarParameter() {defaultValue=1000, parameterName=H}
  n24 = Saturate(None=Multiply(A=Subtract(A=@n18.R, B=@n22), B=@n23))
  n25 = Add(A=@n21, B=Add(A=@n5.Param2, B=ScalarParameter() {defaultValue=0.07, parameterName=EdgeOffset}))
  n26 = Saturate(None=Multiply(A=Subtract(A=@n18.R, B=@n25), B=@n23))
  n27 = Multiply(A=@n5.Param3, B=Saturate(None=Subtract(A=@n24, B=@n26)))
  MaterialFunctionCall(Emissive=Divide(A=LinearInterpolate(A=@n14, B=@n15.RGB, Alpha=@n27))) {materialFunction=&MF_FX_EmissiveAdjustments}
  n28 = Constant() {r=1}
  Multiply(
   A=Multiply(
    A=MaterialFunctionCall!(TexCoord=@n10, ClampChannel=@n28) {materialFunction=/Game/Effects/Fort_Effects/MaterialFunctions/MF_FX_TilingMethod}.Result
    B=@n24
   )
   B=LinearInterpolate(
    A=MaterialFunctionCall(Alpha Threshold=@n13.A) {materialFunction=&DitherTemporalAA}.Result
    B=@n13.A
    Alpha=Floor(None=@n13.A)
   )
  )
  Rotator(
   Coordinate=MaterialFunctionCall(UVs=@n9, Texture Scale=ScalarParameter() {defaultValue=0.5, parameterName=sc}) {materialFunction=&ScaleUVsByCenter}.UVs
   Time=MaterialFunctionCall() {materialFunction=&Pi}.Result
  )
 ]
}
